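# Physics asset for the Red Riding Remix backpack cosmetic mesh, a medium female Fortnite back bling accessory.
# A single capsule body on the pack bone approximates the whole bag, rotated to lie roughly horizontal.
# It has no constraints, so the asset defines no jointed or secondary ragdoll parts.
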PhysicsAsset /BRCosmetics/Accessories/FORT_Backpacks/F_MED_Red_Riding_Remix_Backpack/Meshes/F_MED_Red_Riding_Remix_Pack_PhysicsAsset
bodies {pack [{shapeType=Capsule, length=13.9511, radius=13.947, center=<0.0331 -1.294 -3.819>, rotation=<rot 0.1211 -90 -3.1587>}]}
constraints []
